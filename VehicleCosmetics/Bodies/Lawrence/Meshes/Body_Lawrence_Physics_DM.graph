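# Physics asset for the Lawrence vehicle body mesh in the vehicle cosmetics content.
# It holds a single body on the root bone, shaped as one large box centered forward of and above the bone origin.
# It has no constraints, so the vehicle body is represented as a single simple collision volume.
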PhysicsAsset /VehicleCosmetics/Bodies/Lawrence/Meshes/Body_Lawrence_Physics_DM
bodies {root [{shapeType=Box, extentX=444, extentY=216, extentZ=132, shapeName=root_box, center=<30.75 0 99>}]}
constraints []
